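# Niagara system for the LightSheen vehicle boost cosmetic, trailing ribbons behind the car from skeletal socket positions.
# Two emitters, a colors ribbon and a brighter stripes ribbon, spawn per filtered port and offset left or right by spawn group.
# Painted-variant overrides set color, while game-mode switches and masks adjust offset, lifetime and backward velocity.
# Inherited source movement fades over each particle's life, keeping fresh segments attached to the car.
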NiagaraSystem /VehicleCosmetics/Boosts/Boost_LightSheen/FX/NS_Boost_LightSheen
dependencies {
 usedDataInterfaces [&NiagaraDataInterfaceCurve, &NiagaraDataInterfaceArrayFloat, &NiagaraDataInterfaceSkeletalMesh]
 usedDynamicInputs []
 usedModules [
  &self:SystemSpawnScript.SystemScriptSource.SystemScriptGraph.NiagaraNodeAssignment_0.SetVariables_#1
  &SystemState
 ]
}
emitters {
 NE_ColorsRibbon {
  emitterUpdateScript [
   /SetVariables(Emitter.Shared.SetVariables_#2_Particles.Material Interface=?) {isSetParameters=true}
   &EmitterState()
   &VC_FilteredSocketCount()
   &VC_RateQualityHandler(Epic Spawn Rate=30, High Spawn Rate=30, Low Spawn Rate=30, Use Custom Rate?=true)
   &VC_SpawnRibbon_v1(Number of Ribbons=$Emitter.PortCount, SpawnRate=$Emitter.Spawn Rate)
  ]
  particleSpawnScript [
   &VC_PaintedVariantsOverride(
    Black=<rgba 0 0 0 0>
    BurntSienna=<rgba 0.0642 0.0198 0.0079 0>
    Cobalt=<rgba 0 0.0185 1 0>
    Crimson=<rgba 0.4635 0 0.0126 0>
    DefaultColor=<rgba 0.1559 0.1144 0.6376 0.349>
    ForestGreen=<rgba 0 0.5 0.0253 0>
    Grey=<rgba 0.2031 0.2031 0.2031 0>
    Lime=<rgba 0.2943 0.5885 0 0>
    Orange=<rgba 0.6724 0.2346 0 0>
    Pink=<rgba 0.5052 0.0605 0.3435 0>
    Purple=<rgba 0.1578 0 0.6795 0>
    R1 Black:LinearColor=<rgba 0.015 0.015 0.015 0.35>
    R1 BurntSienna:LinearColor=<rgba 0.1 0.0228 0 0.375>
    R1 Cobalt:LinearColor=<rgba 0.08 0.1277 1 0.45>
    R1 Crimson:LinearColor=<rgba 0.5 0 0.0136 0.28>
    R1 DefaultColor=<rgba 0.8115 0.0076 0.0076 0>
    R1 ForestGreen:LinearColor=<rgba 0 0.4 0.0202 0.35>
    R1 Grey:LinearColor=<rgba 0.125 0.125 0.125 0.28>
    R1 Lime:LinearColor=<rgba 0.2943 0.5885 0 0.35>
    R1 Orange:LinearColor=<rgba 0.6724 0.2346 0 0.28>
    R1 Pink:LinearColor=<rgba 0.5078 0.144 0.6 0.32>
    R1 Purple:LinearColor=<rgba 0.1 0 0.25 0.35>
    R1 Saffron:LinearColor=<rgba 0.5652 0.5652 0 0.35>
    R1 SkyBlue:LinearColor=<rgba 0 0.2724 0.75 0.35>
    R1 TitaniumWhite:LinearColor=<rgba 0.75 0.75 0.75 0.25>
    Saffron=<rgba 0.9 0.69 0 0>
    SkyBlue=<rgba 0 0.357 0.5365 0>
    TitaniumWhite=<rgba 0.75 0.75 0.75 0>
    Use R1 Painted Overrides=true
   )
   &VC_InitializeRibbon_v1(
    Color=$Particles.Color
    Custom Offset=true
    Enable Skeletal Position=true
    Lifetime=/VehicleCosmetics/FX/Niagara/Modules/VC_Mask_Floats(Coro Value:NiagaraFloat=1, FrontEnd Value:NiagaraFloat=1.1, FN Value:NiagaraFloat=0.75)
    Local Vector Offset=&AddVectorToPosition(
     Position=&MergeHelper_VectorSwitch(0=<-26.25 0 0>, 1=<-16.25 0 0>, 2=<-16.25 0 0>, 3=?, Index=$System.GameModeID, Num=::NewEnumerator1)
     Vector=Linked() {linkedVariable {name=User.DTCarOffset, type=NiagaraPosition}}
    )
    Number of Bones Per Boneflow group=2
    Offset Amount=1
    Ribbon Width=1
   )
   &AddVelocity(Velocity=&VC_Mask_Vectors(Coro Value:Vector3f=<-600 0 0>, FrontEnd Value:Vector3f=<-1000 0 0>, FN Value:Vector3f=<-2000 0 0>)) {module=AddVelocity001}
   /VehicleCosmetics/FX/VC_FXParents/Modules/DLR_ProductPreviews(GameModeID:NiagaraInt32=$System.GameModeID, bPreviewVehicleSpeed:NiagaraBool=$System.bPreviewVehicleSpeed)
  ]
  particleUpdateScript [
   &ParticleState()
   &InheritSourceMovement(
    Applied Position Scale=&Multiply_VectorByFloat(
     Vector=<1 1 1>
     Float=&FloatFromCurve(FloatCurve=DataInterface() {propertyValues {Curve {defaultValue=1, keys [{tangentMode=RCTM_User, time=0, value=0.5}, {tangentMode=RCTM_User, time=1, value=1}]}}})
    )
    ApplyForce=false
    ApplyPosition=true
    ApplyPositionFalloff=false
   ) {enabled=false, module=InheritSourceMovement001}
   /SetVariables(
    Particles.LeftRibbon:NiagaraFloat=&SelectFloatFromArray(Direct Array Index=&SpawnGroup(), Float Selection Array=DataInterface() {propertyValues {FloatData [0,85]}})
    Particles.RightRibbon:NiagaraFloat=&SelectFloatFromArray(Direct Array Index=&SpawnGroup(), Float Selection Array=DataInterface() {propertyValues {FloatData [-85,0]}})
   ) {isSetParameters=true, module=SetVariables_#2}
   &InheritSourceMovement(
    Applied Position Scale=&VectorFromFloat(Value=&FloatFromCurve(FloatCurve=DataInterface() {propertyValues {Curve {
         defaultValue=1
         keys [
          {arriveTangent=-1.7885, interpMode=RCIM_Cubic, leaveTangent=-1.7885, tangentMode=RCTM_User, time=0.05, value=1}
          {interpMode=RCIM_Cubic, tangentMode=RCTM_User, time=1, value=0}
         ]
        }}}))
    ApplyForce=false
    ApplyPosition=true
    ApplyPositionFalloff=false
   )
   &ScaleColor(
    Scale Alpha=&FloatFromCurve(FloatCurve=DataInterface() {propertyValues {Curve {defaultValue=1, keys [{tangentMode=RCTM_User, time=0, value=1}, {tangentMode=RCTM_User, time=1, value=1}]}}})
    ScaleA=true
    ScaleRGB=true
   )
   &SolveForcesAndVelocity(Clamp Velocity=false, Limit Acceleration=false)
   &VC_UpdateRibbon_v1(Enable Skeletal Position=true)
  ]
  rendererClasses [&NiagaraRibbonRendererProperties]
  renderers [
   {rendererClass=&NiagaraRibbonRendererProperties}
   {rendererClass=&NiagaraRibbonRendererProperties, rendererIndex=1}
   {rendererClass=&NiagaraRibbonRendererProperties, rendererIndex=2}
   {rendererClass=&NiagaraRibbonRendererProperties, rendererIndex=3}
  ]
 }
 NE_StripesRibbon {
  emitterUpdateScript [
   &EmitterState()
   &VC_FilteredSocketCount()
   &VC_RateQualityHandler(Epic Spawn Rate=30, High Spawn Rate=30, Low Spawn Rate=30, Use Custom Rate?=true)
   &VC_SpawnRibbon_v1(Number of Ribbons=$Emitter.PortCount, SpawnRate=$Emitter.Spawn Rate)
  ]
  particleSpawnScript [
   &VC_PaintedVariantsOverride(
    Black=<rgba 0.002 0.002 0.002 1>
    BurntSienna=<rgba 0.01 0.0042 0.0008 1>
    Cobalt=<rgba 0 0.0148 0.8 2.5>
    Crimson=<rgba 0.4635 0 0.013 2.5>
    DefaultColor=<rgba 0.1144 0.3419 0.9647 0.4>
    ForestGreen=<rgba 0 0.1565 0.0079 2.5>
    Grey=<rgba 0.022 0.022 0.022 1>
    Lime=<rgba 0.1696 0.3391 0 1>
    Orange=<rgba 0.2609 0.1906 0 1>
    Pink=<rgba 0.5052 0.0605 0.3435 1>
    Purple=<rgba 0.005 0 0.01 20>
    Saffron=<rgba 0.7217 0.5533 0 1>
    SkyBlue=<rgba 0 0.6654 1 1.65>
    TitaniumWhite=<rgba 0.75 0.75 0.75 1.5>
   )
   &VC_InitializeRibbon_v1(
    Color=$Particles.Color
    Custom Offset=true
    Enable Skeletal Position=true
    Lifetime=&VC_Mask_Floats(Coro Value:NiagaraFloat=1, FrontEnd Value:NiagaraFloat=1.1, FN Value:NiagaraFloat=1)
    Local Vector Offset=&AddVectorToPosition(
     Position=&MergeHelper_VectorSwitch(0=<-26.5 0 0>, 1=<-16.5 0 0>, 2=<-16.5 0 0>, 3=?, Index=$System.GameModeID, Num=::NewEnumerator1)
     Vector=Linked() {linkedVariable {name=User.DTCarOffset, type=NiagaraPosition}}
    )
    Number of Bones Per Boneflow group=2
    Offset Amount=1
    Ribbon Width=52.5
   )
   &AddVelocity(Velocity=&VC_Mask_Vectors(Coro Value:Vector3f=<-600 0 0>, FrontEnd Value:Vector3f=<-1000 0 0>, FN Value:Vector3f=<-2000 0 0>)) {module=AddVelocity001}
   &DLR_ProductPreviews(GameModeID:NiagaraInt32=$System.GameModeID, bPreviewVehicleSpeed:NiagaraBool=$System.bPreviewVehicleSpeed)
  ]
  particleUpdateScript [
   &ParticleState()
   &InheritSourceMovement(
    Applied Position Scale=&Multiply_VectorByFloat(
     Vector=<1 1 1>
     Float=&FloatFromCurve(FloatCurve=DataInterface() {propertyValues {Curve {defaultValue=1, keys [{tangentMode=RCTM_User, time=0, value=0.5}, {tangentMode=RCTM_User, time=1, value=1}]}}})
    )
    ApplyForce=false
    ApplyPosition=true
    ApplyPositionFalloff=false
   ) {enabled=false, module=InheritSourceMovement001}
   &InheritSourceMovement(
    Applied Position Scale=&VectorFromFloat(Value=&FloatFromCurve(FloatCurve=DataInterface() {propertyValues {Curve {
         defaultValue=1
         keys [
          {arriveTangent=-1.7885, interpMode=RCIM_Cubic, leaveTangent=-1.7885, tangentMode=RCTM_User, time=0.05, value=1}
          {interpMode=RCIM_Cubic, tangentMode=RCTM_User, time=1, value=0}
         ]
        }}}))
    ApplyForce=false
    ApplyPosition=true
    ApplyPositionFalloff=false
   )
   /SetVariables(
    Particles.LeftRibbon:NiagaraFloat=&SelectFloatFromArray(Direct Array Index=&SpawnGroup(), Float Selection Array=DataInterface() {propertyValues {FloatData [0,52.5]}})
    Particles.RightRibbon:NiagaraFloat=&SelectFloatFromArray(Direct Array Index=&SpawnGroup(), Float Selection Array=DataInterface() {propertyValues {FloatData [-52.5,0]}})
   ) {isSetParameters=true}
   &ScaleColor(
    Scale Alpha=&FloatFromCurve(FloatCurve=DataInterface() {propertyValues {Curve {keys [<at 0 1>,<at 1 1>]}}})
    Scale RGB=&Multiply_VectorByFloat(Vector=<1 1 1>, Float=12)
    ScaleA=true
    ScaleRGB=true
   )
   &SolveForcesAndVelocity(Clamp Velocity=false, Limit Acceleration=false)
   &VC_UpdateRibbon_v1(Enable Skeletal Position=true)
  ]
  rendererClasses [&NiagaraRibbonRendererProperties]
  renderers [
   {rendererClass=&NiagaraRibbonRendererProperties}
   {rendererClass=&NiagaraRibbonRendererProperties, rendererIndex=1}
  ]
 }
}
summary {
 systemName=NS_Boost_LightSheen
 emitters [
  {emitterName=NE_StripesRibbon, rendererClasses [&NiagaraRibbonRendererProperties]}
  {emitterName=NE_ColorsRibbon, rendererClasses [&NiagaraRibbonRendererProperties]}
 ]
 userVariables [
  User.bIsFrontendPreview {type=NiagaraBool, defaultValue=true}
  User.bIsPainted {type=NiagaraBool}
  User.ColorVariantID {type=NiagaraInt32}
  User.DTCarOffset {type=NiagaraPosition, description="", defaultValue {}}
  User.DTCarOffset {type=Vector3f, description="", defaultValue {struct=Vector3f, value=<0 0 0>}}
  User.GameModeID {type=NiagaraInt32, defaultValue=3}
  User.PaintedColor {type=LinearColor, description="", defaultValue {struct=LinearColor, value=<rgba 1 1 1 1>}}
  User.PaintedScalar {type=NiagaraFloat, defaultValue=1, description=""}
  User.Particles_MaterialRandom {type=NiagaraFloat, defaultValue=0, description=""}
  User.PortCount {type=NiagaraInt32, defaultValue=2}
  User.RibbonInsideColor {type=NiagaraFloat, defaultValue=1, description=""}
 ]
}
